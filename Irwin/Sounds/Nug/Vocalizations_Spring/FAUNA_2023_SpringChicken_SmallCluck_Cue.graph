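# Sound cue for a small cluck vocalization of the 2023 spring chicken fauna, filed under Nug creature vocalizations.
# A quality-level node sends the first two levels to a six-wave random pool and the last two to a three-wave pool excluded from branch culling.
# Both equally weighted pools include empty slots, giving a one-in-four chance of silence.
# Prey vocalization attenuation, creature sound class and two Nug concurrency sets govern playback.
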SoundCue /Irwin/Sounds/Nug/Vocalizations_Spring/FAUNA_2023_SpringChicken_SmallCluck_Cue
root=QualityLevel_0
nodes [
 {id=QualityLevel_0, children [Random_0,Random_0,Random_1,Random_1]}
 {
  id=Random_0
  children [WavePlayer_0,WavePlayer_1,WavePlayer_2,WavePlayer_3,WavePlayer_4,WavePlayer_5,null,null]
  properties {weights [1,1,1,1,1,1,1,1]}
 }
 {id=Random_1, children [WavePlayer_6,WavePlayer_7,WavePlayer_8,null], properties {bShouldExcludeFromBranchCulling=true, weights [1,1,1,1]}}
 {id=WavePlayer_0, properties {soundWaveAssetPtr=&/FAUNA_2023_SpringChicken_SmallCluck_010}}
 {id=WavePlayer_1, properties {soundWaveAssetPtr=&/FAUNA_2023_SpringChicken_SmallCluck_011}}
 {id=WavePlayer_2, properties {soundWaveAssetPtr=&/FAUNA_2023_SpringChicken_SmallCluck_012}}
 {id=WavePlayer_3, properties {soundWaveAssetPtr=&/FAUNA_2023_SpringChicken_SmallCluck_013}}
 {id=WavePlayer_4, properties {soundWaveAssetPtr=&/FAUNA_2023_SpringChicken_SmallCluck_014}}
 {id=WavePlayer_5, properties {soundWaveAssetPtr=&/FAUNA_2023_SpringChicken_SmallCluck_015}}
 {id=WavePlayer_6, properties {soundWaveAssetPtr=&FAUNA_2023_SpringChicken_SmallCluck_010}}
 {id=WavePlayer_7, properties {soundWaveAssetPtr=&FAUNA_2023_SpringChicken_SmallCluck_011}}
 {id=WavePlayer_8, properties {soundWaveAssetPtr=&FAUNA_2023_SpringChicken_SmallCluck_012}}
]
properties {
 attenuationSettings=/Irwin/Sounds/Settings/Prey_Vocalization_Att
 maxDistance=6350
 priority=0.6
 soundClassObject=/Game/Sounds/FortSoundClassesAndModes/Fort_Creatures_Vocalization_Normal
 concurrencySet [/Irwin/Sounds/Settings/Conc/Nug_Global_VO_Conc, /Irwin/Sounds/Settings/Conc/Nug_SetPriority_VO_Conc]
}
