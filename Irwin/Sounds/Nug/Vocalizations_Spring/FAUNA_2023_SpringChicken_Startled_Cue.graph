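# Sound cue for the startled vocalization of the Spring Chicken fauna variant, filed under Nug sounds.
# A Random node picks one of four equally weighted startled wave variations and is excluded from branch culling.
# Playback uses the shared prey vocalization attenuation and the tactical creature vocalization sound class, with boosted volume.
# Two Nug vocal concurrency sets, one global and one priority-setting, manage overlapping calls.
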SoundCue /Irwin/Sounds/Nug/Vocalizations_Spring/FAUNA_2023_SpringChicken_Startled_Cue
root=Random_0
nodes [
 {
  id=Random_0
  children [WavePlayer_0,WavePlayer_1,WavePlayer_2,WavePlayer_3]
  properties {bShouldExcludeFromBranchCulling=true, weights [1,1,1,1]}
 }
 {id=WavePlayer_0, properties {soundWaveAssetPtr=&/FAUNA_2023_SpringChicken_Startled_010}}
 {id=WavePlayer_1, properties {soundWaveAssetPtr=&/FAUNA_2023_SpringChicken_Startled_011}}
 {id=WavePlayer_2, properties {soundWaveAssetPtr=&/FAUNA_2023_SpringChicken_Startled_012}}
 {id=WavePlayer_3, properties {soundWaveAssetPtr=&/FAUNA_2023_SpringChicken_Startled_013}}
]
properties {
 attenuationSettings=/Irwin/Sounds/Settings/Prey_Vocalization_Att
 maxDistance=6350
 priority=2.5
 soundClassObject=&Fort_Creatures_Vocalization_Tactical
 volumeMultiplier=1.3
 concurrencySet [/Irwin/Sounds/Settings/Conc/Nug_Global_VO_Conc, /Irwin/Sounds/Settings/Conc/Nug_SetPriority_VO_Conc]
}
